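# Unlit debug material in the water splash effect folder, flagged for Niagara mesh particle use, built to visualize vertex color as numeric text.
# DebugFloat3Values takes VertexColor as its value.
# ScreenAlignedPixelToPixelUVs with a fixed resolution feeds Frac, so the readout UVs repeat in screen-space tiles.
# A separate branch appends a constant to the texture coordinates and scales the result; the graph shows no link from either branch to a material output.
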Material /Game/Effects/Niagara/Emitters/Water/SplashEffect/M_DebugVertexColor
graph {
 properties {bUsedWithNanite=true, bUsedWithNiagaraMeshParticles=true, pixelDepthOffsetMode=PDOM_Legacy, shadingModel=MSM_Unlit, usedShadingModels=Unlit}
 nodes [
  n1 = Constant() {r=256}
  MaterialFunctionCall!(
   Vector3=VertexColor()
   MaximumNumberOfDigits
   UVs=Frac(None=MaterialFunctionCall!(TextureResolution=@n1) {materialFunction=/Engine/Functions/Engine_MaterialFunctions02/Texturing/ScreenAlignedPixelToPixelUVs}.Result)
   DebugTextLocation RG_UpperRight BA_LowerLeft
   Component Spacing
  ) {materialFunction=/Engine/Functions/Engine_MaterialFunctions02/Utility/DebugFloat3Values}
  Multiply(A=AppendVector(A=TextureCoordinate(), B=Constant()), B=Constant3Vector() {constant=<rgba 50 50 50 0>})
 ]
}
