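# Translucent unlit Niagara sprite and ribbon material for FLIP water splash particles.
# A soft radial mask, scaled by particle alpha, offsets a scene color lookup to fake refraction, tinted by particle color and blended toward a dynamic-parameter color.
# The mask also drives opacity, zeroed where sampled color is black, including forward and mobile paths.
# Water displacement plus a Z offset moves particles with the water surface on the default path only.
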Material /Game/Effects/Niagara/Emitters/Water/SplashEffect/M_FLIPParticle
graph {
 properties {
  bEnableResponsiveAA=true
  bUsedWithNanite=true
  bUsedWithNiagaraRibbons=true
  bUsedWithNiagaraSprites=true
  blendMode=BLEND_Translucent
  pixelDepthOffsetMode=PDOM_Legacy
  shadingModel=MSM_Unlit
  usedShadingModels=Unlit
 }
 nodes [
  n1 = ParticleColor()
  n2 = TextureCoordinate()
  n3 = Constant() {r=0.5}
  n4 = ScalarParameter() {defaultValue=0.5, parameterName=Radius}
  n5 = Constant()
  n6 = Multiply(A=@n1.A, B=Power(Base=SphereMask(A=@n2, B=@n3, Radius=@n4, Hardness=@n5), Exp=ScalarParameter() {defaultValue=2, parameterName=Hardness}))
  n7 = LinearInterpolate(A=Constant2Vector(), B=Constant2Vector() {g=0.01, r=0.01}, Alpha=@n6)
  n8 = Constant()
  n9 = ShadingPathSwitch(
   Default=Multiply(A=SceneColor!(OffsetFraction=@n7) {inputMode=OffsetFraction, constInput=<0.01 0.01>}.RGB, B=EyeAdaptation().EyeAdaptation)
   Forward=@n8
   Mobile=@n8
  )
  n10 = Constant()
  If(
   A=MaterialFunctionCall(Vector 3=ComponentMask(None=@n9) {b=true, g=true, r=true}) {materialFunction=/Engine/Functions/Engine_MaterialFunctions02/Utility/VectorLength}.V3 Length
   B=@n10
   "A > B"=@n6
   "A == B"=@n10
   "A < B"=@n10
  )
  n11 = DynamicParameter()
  n12 = AppendVector(A=@n11.Param2, B=@n11.Param2)
  MaterialFunctionCall(Emissive=LinearInterpolate(A=Multiply(A=@n9, B=@n1.RGB), B=AppendVector(A=@n12, B=@n11.Param4), Alpha=@n11.Param1)) {materialFunction=&MF_FX_EmissiveAdjustments}
  n13 = MaterialFunctionCall!(Ignore XY Offsets=StaticBoolParameter() {defaultValue=true, parameterName=Ignore XY Gerstner Offsets}, Position, WaterBodyIndex, Depth) {materialFunction=/FortWater/Materials/Functions/WaterParticleDisplacement}
  n14 = VectorParameter() {parameterName=Water Displacement Multiplier, defaultValue=<rgba 1 1 1 0>}
  n15 = ScalarParameter() {defaultValue=-1, parameterName=ZOffset}
  n16 = Constant()
  ShadingPathSwitch(
   Default=Add(
    A=Multiply(A=@n13.WPO, B=@n14.RGB)
    B=MaterialFunctionCall(Z=@n15) {materialFunction=&MakeFloat3}.Result
   )
   Forward=@n16
   Mobile=@n16
  )
 ]
}
