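# Material function for Niagara water splash materials that applies a final opacity and emissive control pass to incoming material attributes.
# A tiled, panning alpha mask, multiplied by a channel-select vector and desaturated, scales opacity.
# A separately tiled, panning color map, multiplied by a tint, modulates emissive color.
# The EnableOverallControl static switch chooses between this modified result and the untouched input, with all settings exposed as OverallControl parameters.
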MaterialFunction /Game/Effects/Niagara/Emitters/Water/SplashEffect/MF_WaterFinalControl
graph {nodes [
  n1 = FunctionInput() {inputName=InMA, inputType=FunctionInput_MaterialAttributes}
  n2 = GetMaterialAttributes(None=@n1) {attributeGetTypes [#1, #2]}
  n3 = TextureCoordinate()
  n4 = AppendVector(
   A=ScalarParameter() {group=OverallControl, parameterName=MF_OverallControl_AlphaMaskTilling_X}
   B=ScalarParameter() {group=OverallControl, parameterName=MF_OverallControl_AlphaMaskTilling_Y}
  )
  n5 = ScalarParameter() {group=OverallControl, parameterName=MF_OverallControl_AlphaMaskPanner_X}
  n6 = ScalarParameter() {group=OverallControl, parameterName=MF_OverallControl_AlphaMaskPanner_Y}
  n7 = TextureSampleParameter2D(UVs=Panner(Coordinate=Multiply(A=@n3, B=@n4), Time=Time(), Speed=AppendVector(A=@n5, B=@n6))) {group=OverallControl, parameterName=MF_OverallControl_AlphaMask, texture=/Game/Global/Textures/Default/Noise/TankMoteNoise}
  n8 = VectorParameter() {group=OverallControl, parameterName=MF_OverallControl_AlphaMaskSelect, defaultValue=<rgba 1 0 0 0>}
  n9 = AppendVector(A=@n8.RGB, B=@n8.RGB)
  n10 = Multiply(A=@n2.Opacity, B=Desaturation(None=Multiply(A=@n7.RGBA, B=@n9)))
  n11 = TextureCoordinate()
  n12 = AppendVector(
   A=ScalarParameter() {group=OverallControl, parameterName=MF_OverallControl_ColorMapTilling_X}
   B=ScalarParameter() {group=OverallControl, parameterName=MF_OverallControl_ColorMapTilling_Y}
  )
  n13 = ScalarParameter() {group=OverallControl, parameterName=MF_OverallControl_ColorMapPanner_X}
  n14 = ScalarParameter() {group=OverallControl, parameterName=MF_OverallControl_ColorMapPanner_Y}
  n15 = TextureSampleParameter2D(UVs=Panner(Coordinate=Multiply(A=@n11, B=@n12), Time=Time(), Speed=AppendVector(A=@n13, B=@n14))) {
   group=OverallControl
   parameterName=MF_OverallControl_ColorMap
   samplerType=SAMPLERTYPE_LinearColor
   texture=&HighContrastNoise
  }
  n16 = Multiply(
   A=Multiply(A=@n15.RGB, B=VectorParameter() {group=OverallControl, parameterName=MF_OverallControl_ColorMapTint, defaultValue=<rgba 1 0 0 0>}.RGB)
   B=@n2.EmissiveColor
  )
  FunctionOutput(None=StaticSwitchParameter(
    True=SetMaterialAttributes!(MaterialAttributes=@n2.MaterialAttributes, Opacity=@n10, EmissiveColor=@n16) {attributeSetTypes [#1, #2]}
    False=@n1
   ) {group=OverallControl, parameterName=EnableOverallControl}) {outputName=OutMA}
 ]}
