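# Translucent subsurface water splash material for Niagara sprite, ribbon and mesh particles, blending scrolling water and foam layers.
# Dynamic parameters and curve-atlas rows animate the flow, layer blend, eroded edge-masked opacity and mesh width.
# World position offset shapes camera- or Z-facing strips with drop, distance scaling and optional FortWater wave displacement.
# Cubemap reflection, distance-faded refraction, depth and camera fades, biome tint and particle-driven HSV offset finish the look.
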Material /Game/Effects/Niagara/Emitters/Water/SplashEffect/M_NFX_WaterSplash
graph {
 properties {
  bCastRayTracedShadows=false
  bUseMaterialAttributes=true
  bUsedWithMeshParticles=true
  bUsedWithNanite=true
  bUsedWithNiagaraMeshParticles=true
  bUsedWithNiagaraRibbons=true
  bUsedWithNiagaraSprites=true
  blendMode=BLEND_Translucent
  pixelDepthOffsetMode=PDOM_Legacy
  refractionMethod=RM_IndexOfRefraction
  shadingModel=MSM_Subsurface
  twoSided=true
  usedShadingModels=Subsurface
 }
 nodes [
  Add()
  n1 = Multiply(
   A=StaticSwitchParameter(
    True=Time()
    False=DynamicParameter() {paramNames [MainFlowAnimation, MainTextureAnimation, MainWPOscale, WPO_Zaix_Drop]}.MainTextureAnimation
   ) {defaultValue=true, parameterName=PreviewPime}
   B=ScalarParameter() {defaultValue=1, group=Animation, parameterName=MasterTextureMotion}
  ) {constB=-0.5}
  AppendVector(B=@n1)
  Constant()
  Constant() {r=1}
  Constant()
  Constant()
  Constant() {r=1}
  Constant() {r=0.5}
  n2 = TextureCoordinate()
  n3 = StaticSwitchParameter(
   True=MaterialFunctionCall(XY=OneMinus(None=@n2)) {materialFunction=&Swizzle}.YX
   False=@n2
  ) {parameterName=SwizzleUV?}
  n4 = MaterialFunctionCall(Float2=@n3) {materialFunction=&BreakOutFloat2Components}
  n5 = TextureSample(UVs=@n4.G) {
   samplerSource=SSM_Wrap_WorldGroupSettings
   samplerType=SAMPLERTYPE_LinearColor
   texture=/Game/Effects/Niagara/MasterMaterials/Liquid/RT_Tex18
  }
  n6 = ShadingPathSwitch(Default=@n5.R, Mobile=StaticSwitchParameter(True=@n5.R, False=Constant()) {parameterName=U Shape UV on Low}) {desc=Migrated from FeatureLevelSwitch}
  n7 = ScalarParameter() {defaultValue=-0.5, group=OpacityMaskControl, parameterName=U_Shape}
  n8 = AppendVector(A=Add(A=@n4.R, B=Multiply(A=@n6, B=@n7) {constB=0.25}), B=@n4.G)
  n9 = Multiply(
   A=DynamicParameter() {paramNames [MainFlowAnimation, MainTextureAnimation, MainWPOscale, WPO_Zaix_Drop]}.MainFlowAnimation
   B=ScalarParameter() {defaultValue=-0.5, parameterName=TimeScale}
  ) {constB=-0.5}
  n10 = VertexColor()
  n11 = LinearInterpolate(A=ScalarParameter() {defaultValue=1.3, parameterName=VertexRedColorAffectTimeScale}, Alpha=@n10.R)
  n12 = ScalarParameter() {defaultValue=0.2, group=Animation, parameterName=FlowAnimationNoise_U_Repeat}
  n13 = ScalarParameter() {defaultValue=1, group=Animation, parameterName=FlowAnimationNoise_V_Repeat}
  n14 = DynamicParameter() {parameterIndex=1, defaultValue=<rgba 0 0 0 0>, paramNames [Erode, AlphaMaskVflip, MakeParticlesFaceZaxis, ""]}
  n15 = LinearInterpolate(A=@n13, B=Multiply(A=@n13) {constB=-1}, Alpha=@n14.AlphaMaskVflip)
  n16 = TextureSample(UVs=Multiply(A=@n3, B=AppendVector(A=@n12, B=@n15))) {
   samplerSource=SSM_Wrap_WorldGroupSettings
   samplerType=SAMPLERTYPE_LinearColor
   texture=&RT_Tex18
  }
  n17 = Multiply(A=@n16.G, B=ScalarParameter() {defaultValue=0.15, group=Animation, parameterName=FlowAnimationNoise_Intensity})
  n18 = Constant()
  n19 = Add(
   A=ComponentMask(None=Multiply(A=@n8) {constB=0.33}) {r=true}
   B=Add(A=Multiply(A=@n9, B=@n11), B=ShadingPathSwitch(Default=@n17, Mobile=@n18) {desc=Migrated from FeatureLevelSwitch}) {constB=0}
  )
  CurveAtlasRowParameter(CurveTime=Frac(None=@n19)) {
   atlas=/Game/Effects/Niagara/Emitters/Water/WaterAtlas
   curve=&/Curve_WaterSplashCurve2
   defaultValue=1
   group=Animation
   parameterName=AnimationTemp
  }
  n20 = VectorParameter() {group=WaterShader, parameterName=UVS, defaultValue=<rgba 0.2 0.4 -0.3 0.4>}
  n21 = Add(A=@n1, B=@n10.R)
  n22 = MaterialFunctionCall(Float2=@n3) {materialFunction=&BreakOutFloat2Components}
  n23 = CurveAtlasRowParameter(CurveTime=@n19) {
   atlas=&WaterAtlas
   curve=&/Curve_WaterSplashCurve
   group=Animation
   parameterName=MaterialPannerMasks
  }
  n24 = Multiply(A=AppendVector(A=@n22.R, B=Subtract(A=@n22.G) {constB=0.5}), B=@n23.A)
  n25 = Multiply(A=@n24, B=ComponentMask(None=@n20.RGB) {g=true, r=true})
  n26 = TextureSample(UVs=@n25) {
   samplerSource=SSM_Wrap_WorldGroupSettings
   samplerType=SAMPLERTYPE_Normal
   texture=/Game/Global/Textures/Default/Tileable_Patterns/Cloud_normals
  }
  n27 = AppendVector(A=@n26.R, B=@n26.R)
  n28 = ScalarParameter() {defaultValue=0.04, group=WaterShader, parameterName=Water_Distort}
  n29 = Add(A=@n25, B=ShadingPathSwitch(Default=Multiply(A=@n27, B=@n28), Mobile=Constant()) {desc=Migrated from FeatureLevelSwitch})
  n30 = Add(A=@n29, B=Multiply(A=AppendVector(A=@n20.B, B=@n20.B), B=@n21))
  n31 = TextureSample(UVs=@n30) {
   samplerSource=SSM_Wrap_WorldGroupSettings
   samplerType=SAMPLERTYPE_LinearColor
   texture=/Game/Global/Textures/Default/Tileable_Patterns/Cloud_Diffuse
  }
  n32 = TextureSample(UVs=Multiply(A=@n29) {constB=0.618}) {
   samplerSource=SSM_Wrap_WorldGroupSettings
   samplerType=SAMPLERTYPE_LinearColor
   texture=&Cloud_Diffuse
  }
  n33 = LinearInterpolate(
   A=LinearInterpolate(A=@n31.R, B=@n32.R)
   B=OneMinus(None=ComponentMask(None=@n3) {r=true})
   Alpha=ScalarParameter() {defaultValue=0.2, group=WaterShader, parameterName=Water_LerpRamp}
  )
  CurveAtlasRowParameter(CurveTime=@n33) {
   atlas=&WaterAtlas
   curve=&/Curve_WaterShaderTint
   defaultValue=2
   group=WaterShader
   parameterName=WaterTint
  }
  n34 = VectorParameter() {group=FoamShader, parameterName=UVS2, defaultValue=<rgba 2 0.5 -0.3 0.4>}
  n35 = TextureSample(UVs=Multiply(A=@n24, B=ComponentMask(None=@n34.RGB) {g=true, r=true})) {
   samplerSource=SSM_Wrap_WorldGroupSettings
   samplerType=SAMPLERTYPE_Normal
   texture=/Game/Effects/Fort_Effects/Textures/Water/T_MaskeSplashes_Bubbles_N
  }
  n36 = LinearInterpolate(
   A=Constant3Vector() {constant=<rgba 0 0 1 0>}
   B=@n35.RGB
   Alpha=ScalarParameter() {defaultValue=0.3, group=FoamShader, parameterName=FroamNormalIntensity}
  )
  n37 = LinearInterpolate(A=@n23.B, B=DotProduct(A=@n36, B=CameraVectorWS())) {constAlpha=0.75}
  CurveAtlasRowParameter(CurveTime=@n37) {
   atlas=&WaterAtlas
   curve=&/Curve_FoamShaderTint
   defaultValue=3
   group=FoamShader
   parameterName=FoamTint
  }
  n38 = ObjectPositionWS()
  LinearInterpolate(A=@n38, B=WorldPosition().XYZ, Alpha=ScalarParameter() {parameterName=WPO_UsingPerVertexSrtip_0or1})
  MakeMaterialAttributes!(
   BaseColor
   Metallic
   Specular
   Roughness
   Anisotropy
   EmissiveColor
   Opacity
   OpacityMask
   Normal
   Tangent
   WorldPositionOffset=Constant()
   SubsurfaceColor
   ClearCoat
   ClearCoatRoughness
   AmbientOcclusion
   Refraction
   CustomizedUVs_0
   CustomizedUVs_1
   CustomizedUVs_2
   CustomizedUVs_3
   CustomizedUVs_4
   CustomizedUVs_5
   CustomizedUVs_6
   CustomizedUVs_7
   PixelDepthOffset
   ShadingModel
   Displacement
  )
  n39 = CurveAtlasRowParameter(CurveTime=ComponentMask(None=@n3) {r=true}) {
   atlas=&WaterAtlas
   curve=&Curve_WaterSplashCurve2
   defaultValue=1
   group=WPO_MeshShape
   parameterName=Z_FalloffCurve
  }
  n40 = Constant()
  n41 = MaterialFunctionCall(
   X=@n40
   Y=@n40
   Z=Multiply(
    A=ScalarParameter() {defaultValue=-100, group=WPO_MeshShape, parameterName=WPO_Zaix_Drop}
    B=DynamicParameter() {paramNames [MainFlowAnimation, MainTextureAnimation, MainWPOscale, WPO_Zaix_Drop]}.WPO_Zaix_Drop
   )
  ) {materialFunction=&MakeFloat3}
  n42 = CurveAtlasRowParameter(CurveTime=ComponentMask(None=@n3) {g=true}) {
   atlas=&WaterAtlas
   curve=&Curve_WaterSplashCurve2
   defaultValue=1
   group=WPO_MeshShape
   parameterName=Z_FalloffCurve
  }
  n43 = Constant()
  n44 = Transform(None=Constant3Vector() {constant=<rgba 1 0 0 0>})
  n45 = Normalize(VectorInput=Subtract(A=@n38, B=CameraPositionWS()))
  n46 = Multiply(
   A=MaterialFunctionCall(X=@n43, Y=@n43, Z=ScalarParameter() {defaultValue=25, group=WPO_MeshShape, parameterName=WPO_Zaix_MiddleCurve}) {materialFunction=&MakeFloat3}.Result
   B=Abs(None=DotProduct(A=@n44, B=@n45))
  )
  n47 = ComponentMask(None=@n3) {g=true}
  n48 = Normalize(VectorInput=Constant3Vector() {constant=<rgba 0 0 1 0>})
  n49 = DynamicParameter() {parameterIndex=1, defaultValue=<rgba 0 0 0 0>, paramNames [Erode, AlphaMaskVflip, MakeParticlesFaceZaxis, ""]}
  n50 = CrossProduct(A=@n44, B=LinearInterpolate(A=@n45, B=@n48, Alpha=@n49.MakeParticlesFaceZaxis))
  n51 = Multiply(A=Multiply(A=Normalize(VectorInput=@n50), B=Subtract(A=@n47) {constB=0.5}), B=@n23.A)
  n52 = ScalarParameter() {defaultValue=200, group=WPO_MeshShape, parameterName=WPO_width}
  n53 = ScalarParameter() {defaultValue=1, group=WPO_DistanceShink, parameterName=WPO_Shink_NearDistanceSIze}
  n54 = ScalarParameter() {defaultValue=1, group=WPO_DistanceShink, parameterName=WPO_Shink_FarDistanceSIze}
  n55 = MaterialFunctionCall(
   Input=Distance(A=ObjectPositionWS(), B=CameraPositionWS())
   Input Low=ScalarParameter() {defaultValue=300, group=WPO_DistanceShink, parameterName=WPO_Shink_NearDistance}
   Input High=ScalarParameter() {defaultValue=5000, group=WPO_DistanceShink, parameterName=WPO_Shink_FarDistance}
   Target Low=@n53
   Target High=@n54
  ) {materialFunction=&RemapValueRange}
  n56 = Add(
   A=Add(A=Multiply(A=@n39.A, B=@n41.Result), B=Multiply(A=@n42.G, B=@n46))
   B=Multiply(A=Multiply(A=@n51, B=@n52), B=Clamp(None=@n55.Result, Min=@n53, Max=@n54))
  )
  n57 = DynamicParameter() {paramNames [MainFlowAnimation, MainTextureAnimation, MainWPOscale, WPO_Zaix_Drop]}
  n58 = VectorParameter() {parameterName=WaterTint A, defaultValue=<rgba 0.7098 0.9804 1 0>}
  n59 = VectorParameter() {parameterName=WaterTint B, defaultValue=<rgba 0.0157 0.2549 0.302 0.89>}
  n60 = VectorParameter() {parameterName=WaterTint C, defaultValue=<rgba 0.0078 0.2078 0.3961 1.1>}
  n61 = MaterialFunctionCall!(
   1=MaterialFunctionCall(X=@n58.R, Y=@n58.R, Z=@n58.R, A=@n58.R) {materialFunction=&MakeFloat4}.Result
   2=MaterialFunctionCall(X=@n59.R, Y=@n59.R, Z=@n59.R, A=@n59.R) {materialFunction=&MakeFloat4}.Result
   3=MaterialFunctionCall(X=@n60.R, Y=@n60.R, Z=@n60.R, A=@n60.R) {materialFunction=&MakeFloat4}.Result
   4
   A=@n33
  ) {materialFunction=/Engine/Functions/Engine_MaterialFunctions02/Lerp_Multiple_Float4}
  n62 = ComponentMask(None=@n61.Lerp 3 Inputs) {b=true, g=true, r=true}
  n63 = ComponentMask(None=@n61.Lerp 3 Inputs) {a=true}
  n64 = TextureSample(UVs=@n30) {
   samplerSource=SSM_Wrap_WorldGroupSettings
   samplerType=SAMPLERTYPE_Normal
   texture=&Cloud_normals
  }
  n65 = VectorParameter() {parameterName=FoamTint A, defaultValue=<rgba 0.7412 0.6353 0.4118 1>}
  n66 = VectorParameter() {parameterName=FoamTint B, defaultValue=<rgba 1 1 1 1.4>}
  n67 = VectorParameter() {parameterName=FoamTint C, defaultValue=<rgba 1.2 1.2 1.2 1.5>}
  n68 = MaterialFunctionCall(
   1=MaterialFunctionCall(X=@n65.R, Y=@n65.R, Z=@n65.R, A=@n65.R) {materialFunction=&MakeFloat4}.Result
   2=MaterialFunctionCall(X=@n66.R, Y=@n66.R, Z=@n66.R, A=@n66.R) {materialFunction=&MakeFloat4}.Result
   3=MaterialFunctionCall(X=@n67.R, Y=@n67.R, Z=@n67.R, A=@n67.R) {materialFunction=&MakeFloat4}.Result
   A=@n37
  ) {materialFunction=&Lerp_Multiple_Float4}
  n69 = ComponentMask(None=@n68.Lerp 3 Inputs) {b=true, g=true, r=true}
  n70 = ComponentMask(None=@n68.Lerp 3 Inputs) {a=true}
  n71 = ScalarParameter() {defaultValue=50, group=OpacityMaskControl, parameterName=DepthFadeHigh}
  n72 = ScalarParameter() {defaultValue=50, group=OpacityMaskControl, parameterName=DepthFadeLow}
  n73 = DynamicParameter() {parameterIndex=1, defaultValue=<rgba 0 0 0 0>, paramNames [Erode, AlphaMaskVflip, MakeParticlesFaceZaxis, ""]}
  n74 = LinearInterpolate(Alpha=@n73.AlphaMaskVflip) {constA=1, constB=-1}
  n75 = Multiply(A=@n3, B=AppendVector(A=@n74, B=@n74))
  n76 = TextureSampleParameter2D(UVs=Panner(Coordinate=StaticSwitchParameter(True=@n75, False=@n3) {parameterName=ErosionTextureRandomFlip}, Time=@n21) {speedX=-0.2}) {
   group=OpacityMaskControl
   parameterName=DepthFadeTextureMask
   samplerSource=SSM_Wrap_WorldGroupSettings
   samplerType=SAMPLERTYPE_LinearColor
   texture=&HighContrastNoise
  }
  n77 = ComponentMask(None=StaticComponentMaskParameter(None=@n76.RGBA) {defaultR=true, group=OpacityMaskControl, parameterName=DepthFadeTextureMask_1ch}) {r=true}
  n78 = OneMinus(None=SphereMask(A=WorldPosition().XYZ, B=CameraPositionWS(), Radius=ScalarParameter() {defaultValue=50, group=OpacityMaskControl, parameterName=CameraFade}) {hardnessPercent=0})
  n79 = Min(A=DepthFade(FadeDistance=LinearInterpolate(A=@n71, B=@n72, Alpha=@n77)), B=@n78)
  n80 = StaticSwitchParameter(True=@n79, False=@n78) {parameterName=Depth Blend on Low}
  n81 = ShadingPathSwitch(Default=@n79, Mobile=@n80) {desc=Migrated from FeatureLevelSwitch}
  n82 = QualitySwitch(Default=@n81, Low=@n80, High=@n81, Medium=@n81)
  n83 = OneMinus(None=@n82)
  n84 = Saturate(None=Add(A=Multiply(A=@n83) {constB=0.5}, B=@n23.R))
  n85 = GetMaterialAttributes(None=BlendMaterialAttributes(
    A=SetMaterialAttributes!(MaterialAttributes, Emissive Color=@n62, Opacity=@n63, Normal=@n64.RGB) {attributeSetTypes [#1, #2, #3]}
    B=SetMaterialAttributes(Emissive Color=@n69, Opacity=@n70, Normal=@n36) {attributeSetTypes [#1, #2, #3]}
    Alpha=@n84
   ) {vertexAttributeBlendType=UseB}) {attributeGetTypes [#2]}
  n86 = ScalarParameter() {parameterName=Angle Fade Bias}
  n87 = Multiply(A=Multiply(A=@n82, B=@n82), B=@n23.G)
  n88 = DynamicParameter() {parameterIndex=1, defaultValue=<rgba 0 0 0 0>, paramNames [Erode, AlphaMaskVflip, MakeParticlesFaceZaxis, ""]}
  n89 = Subtract(A=Add(A=@n76.G) {constB=0.001}, B=@n88.Erode) {constB=0.5}
  n90 = ScalarParameter() {defaultValue=1000, group=OpacityMaskControl, parameterName=ErosionContrast}
  n91 = ShadingPathSwitch(
   Default=Saturate(None=Multiply(A=@n89, B=@n90) {constB=1000})
   Mobile=ScalarParameter() {defaultValue=0.5, parameterName=Erosion Default Mobile}
  ) {desc=Migrated from FeatureLevelSwitch}
  n92 = Multiply(
   A=@n85.Opacity
   B=Multiply(
    A=MaterialFunctionCall("FadeSensitivity (optional)"=@n86) {materialFunction=&FadeBasedOnViewAngle}.Result
    B=Multiply(A=@n87, B=@n91)
   )
  )
  n93 = TextureObjectParameter() {
   group=OpacityMaskControl
   parameterName=EdgeMask_U
   samplerType=SAMPLERTYPE_LinearColor
   texture=&/RT_Tex8
  }
  n94 = TextureSample(UVs=Panner(Coordinate=@n3, Time=@n21) {speedX=-0.2}, Tex=@n93) {
   samplerSource=SSM_Wrap_WorldGroupSettings
   samplerType=SAMPLERTYPE_LinearColor
   texture=&RT_Tex8
  }
  n95 = TextureSample(UVs=Panner(Coordinate=@n3, Time=@n21) {speedX=-0.4}, Tex=@n93) {
   samplerSource=SSM_Wrap_WorldGroupSettings
   samplerType=SAMPLERTYPE_LinearColor
   texture=&RT_Tex8
  }
  n96 = MaterialFunctionCall(Float3=Multiply(A=@n94.RGB, B=@n95.RGB)) {materialFunction=&BreakOutFloat3Components}
  n97 = Multiply(A=@n96.G, B=@n96.G)
  n98 = TextureSampleParameter2D(UVs=@n3) {
   group=OpacityMaskControl
   parameterName=EdgeMask_V
   samplerSource=SSM_Wrap_WorldGroupSettings
   samplerType=SAMPLERTYPE_LinearColor
   texture=/Game/Effects/Niagara/MasterMaterials/Liquid/RT_Tex12
  }
  n99 = Multiply(A=Multiply(A=LinearInterpolate(A=@n96.B, B=@n97, Alpha=@n23.R), B=@n23.G), B=@n98.G)
  n100 = ScalarParameter() {defaultValue=1, group=OpacityMaskControl, parameterName=MasterWaterOpacity}
  n101 = ScalarParameter() {defaultValue=1, group=OpacityMaskControl, parameterName=MasterFoamOpacity}
  n102 = GetMaterialAttributes(None=SetMaterialAttributes!(MaterialAttributes=@n85.MaterialAttributes, Opacity=Multiply(A=@n92, B=@n99), Specular=LinearInterpolate(A=@n100, B=@n101, Alpha=@n84)) {attributeSetTypes [#2, #4]}) {attributeGetTypes [#1, #3]}
  n103 = Constant3Vector() {constant=<rgba 0 0 1 0>}
  n104 = ScalarParameter() {defaultValue=0.05, group=Emissives, parameterName=ReflectionNormal}
  n105 = TextureSampleParameterCube!(UVs=ReflectionVectorWS(CustomWorldNormal=LinearInterpolate(A=@n103, B=@n102.Normal, Alpha=@n104)), Apply View MipBias) {
   group=Emissives
   parameterName=RefractionMap
   texture=/Game/Packages/Fortress_Sky/SkyDome/MaterialInstances/SkyDome_Apollo/Cubemap_Apollo_Day
  }
  n106 = Multiply(A=ScalarParameter() {defaultValue=1, group=Emissives, parameterName=ReflectionIntensity}, B=@n105.RGB)
  n107 = GetMaterialAttributes(None=SetMaterialAttributes!(MaterialAttributes=@n102.MaterialAttributes, Emissive Color=Add(A=@n102.Emissive Color, B=@n106)) {attributeSetTypes [#1]}) {attributeGetTypes [#2]}
  n108 = ScalarParameter() {defaultValue=2, group=OpacityMaskControl, parameterName=OpacityContrast_Water}
  n109 = ScalarParameter() {defaultValue=3, group=OpacityMaskControl, parameterName=OpacityContrast_Foam}
  n110 = SetMaterialAttributes!(
   MaterialAttributes=@n107.MaterialAttributes
   Opacity=MaterialFunctionCall(In=@n107.Opacity, Contrast=LinearInterpolate(A=@n108, B=@n109, Alpha=@n23.R)) {materialFunction=&CheapContrast}.Result
  ) {attributeSetTypes [#2]}
  n111 = StaticSwitchParameter(
   True=@n110
   False=SetMaterialAttributes!(MaterialAttributes=@n110, World Position Offset=Multiply(A=@n56, B=@n57.MainWPOscale)) {attributeSetTypes [#5]}
  ) {parameterName=DisableWPO}
  n112 = GetMaterialAttributes(None=@n111) {attributeGetTypes [#5]}
  n113 = MaterialFunctionCall!(Ignore XY Offsets=StaticBoolParameter() {defaultValue=true, parameterName=Ignore XY Gerstner Offsets}, Position, WaterBodyIndex, Depth) {materialFunction=/FortWater/Materials/Functions/WaterParticleDisplacement}
  n114 = VectorParameter() {parameterName=Water Displacement Multiplier, defaultValue=<rgba 1 1 1 0>}
  n115 = ScalarParameter() {defaultValue=1, parameterName=ZOffset}
  n116 = Add(
   A=@n112.World Position Offset
   B=Add(
    A=Multiply(A=@n113.WPO, B=@n114.RGB)
    B=MaterialFunctionCall(Z=@n115) {materialFunction=&MakeFloat3}.Result
   )
  )
  n117 = GetMaterialAttributes(None=StaticSwitchParameter(
    True=SetMaterialAttributes(MaterialAttributes=@n112.MaterialAttributes, World Position Offset=@n116) {attributeSetTypes [#5]}
    False=@n111
   ) {parameterName=UseWaterDisplacement}) {attributeGetTypes [#5]}
  n118 = MaterialFunctionCall(Offset Amount=ScalarParameter() {defaultValue=1, group=WPO_MeshShape, parameterName=CameraOffset}) {materialFunction=/Engine/Functions/Engine_MaterialFunctions02/WorldPositionOffset/CameraOffset}
  n119 = GetMaterialAttributes(None=SetMaterialAttributes(MaterialAttributes=@n117.MaterialAttributes, World Position Offset=Add(A=@n117.World Position Offset, B=@n118.Clamped Camera Offset)) {attributeSetTypes [#5]}) {attributeGetTypes [#2]}
  n120 = ScalarParameter() {defaultValue=0.98, parameterName=Refraction}
  n121 = Constant() {desc=e, r=2.7183}
  n122 = Multiply(A=Distance(A=WorldPosition().XYZ, B=CameraPositionWS()), B=ScalarParameter() {defaultValue=0.0005, parameterName=RefractionFalloffDensity}) {constB=0.0001}
  n123 = Multiply(A=@n122, B=@n122) {constB=0.0001}
  n124 = Multiply(A=Saturate(None=@n119.Opacity), B=Divide(A=Constant() {r=1}, B=Power(Base=@n121, Exp=@n123)))
  n125 = GetMaterialAttributes(None=SetMaterialAttributes!(MaterialAttributes=@n119.MaterialAttributes, Refraction=LinearInterpolate(B=@n120, Alpha=@n124) {constA=1, constB=1.01}) {attributeSetTypes [#6]}) {attributeGetTypes [#4, #2]}
  n126 = Multiply(A=@n125.Specular, B=@n125.Specular)
  n127 = GetMaterialAttributes(None=MaterialFunctionCall!(InMA=SetMaterialAttributes!(MaterialAttributes=@n125.MaterialAttributes, Opacity=@n126) {attributeSetTypes [#2]}) {materialFunction=&/MF_WaterFinalControl}.OutMA) {attributeGetTypes [
    #1
    #2
    #3
    #5
    #6
   ]}
  n128 = Multiply(A=@n127.Emissive Color, B=VectorParameter() {parameterName=MasterEmissiveTint, defaultValue=<rgba 1 1 1 0>}.RGB)
  n129 = ParticleColor()
  n130 = ScalarParameter() {desc="Interpret Particle RGB as HSV Offsets, and use ParticleColor Fallback instead..", parameterName=Use Particle Color RGB as HSV Offset}
  n131 = LinearInterpolate(A=@n129.RGB, B=VectorParameter() {parameterName=ParticleColor Fallback, defaultValue=<rgba 1 1 1 1>}.RGB, Alpha=@n130)
  n132 = Multiply(A=@n127.Opacity, B=@n129.A)
  n133 = NamedRerouteUsage()
  n134 = GetMaterialAttributes(None=MakeMaterialAttributes(
    BaseColor=Multiply(A=@n128, B=@n131)
    Specular=Constant() {r=1}
    Roughness=Constant()
    EmissiveColor=MaterialFunctionCall!(EmissiveOvrride) {materialFunction=/Game/Effects/Niagara/Emitters/Water/FoamEffect/MF_NFX_EmissiveMask}.Result
    Opacity=@n132
    OpacityMask=MaterialFunctionCall(Alpha Threshold=@n132) {materialFunction=&DitherTemporalAA}.Result
    Normal=@n127.Normal
    WorldPositionOffset=Add(A=@n127.World Position Offset, B=@n133)
    Refraction=@n127.Refraction
   )) {attributeGetTypes [#7, #2, #1]}
  n135 = VectorParameter() {group=WaterShader, parameterName=BiomeColorTint, defaultValue=<rgba 0.0938 0.0907 0.0469 2>}
  n136 = MaterialFunctionCall() {materialFunction=/Game/Effects/Niagara/Emitters/Water/FoamEffect/MF_NFX_BiomeMask}
  n137 = LinearInterpolate(A=@n134.Base Color, B=Multiply(A=ComponentMask(None=@n134.Base Color) {g=true}, B=@n135.RGB), Alpha=@n136.Result) {constA=1}
  n138 = ScalarParameter() {group=Emissives, parameterName=OverallEmissiveAdd}
  n139 = MaterialFunctionCall(Emissive=Add(A=Multiply(A=@n137, B=@n138), B=@n134.Emissive Color)) {materialFunction=&MF_FX_EmissiveAdjustments}
  n140 = LinearInterpolate(A=@n134.Opacity, B=Multiply(A=@n134.Opacity, B=@n135.A), Alpha=@n136.Result) {constA=1, constB=3}
  n141 = ScalarParameter() {desc="Interpret Particle RGB as HSV Offsets, and use ParticleColor Fallback instead..", parameterName=Use Particle Color RGB as HSV Offset}
  MaterialFunctionCall!(
   MA=MaterialFunctionCall!(
    MA=SetMaterialAttributes!(MaterialAttributes=@n134.MaterialAttributes, Base Color=@n137, Opacity=@n140, Emissive Color=@n139.Result) {attributeSetTypes [#7, #2, #1]}
    SSS_TintOvrride
   ) {materialFunction=/Game/Effects/Niagara/Emitters/Water/FoamEffect/MF_NFX_BaseColorToSSS}.MA
   Reference Color
   HSV Offset=LinearInterpolate(A=Constant3Vector(), B=ParticleColor().RGB, Alpha=@n141)
  ) {desc=colored water, materialFunction=/Game/Effects/Niagara/Emitters/Water/FoamEffect/MF_NFX_HSV_Offset}
  n142 = DynamicParameter() {parameterIndex=2, paramNames [VelX,VelY,VelZ,VelInfluence]}
  n143 = ComponentMask(None=TextureCoordinate()) {r=true}
  n144 = ScalarParameter() {defaultValue=2, parameterName=VelWPOMask Pow}
  Multiply(A=@n142.VelInfluence, B=Multiply(A=Power(Base=@n143, Exp=@n144), B=Multiply(B=@n142.RGB) {constA=-1}))
  NamedRerouteDeclaration(None=@n23.G) {name=test, nodeColor=<rgba 0 1 0.5313 1>}
  n145 = DynamicParameter() {desc=Parameters set by Niagara, parameterIndex=2, paramNames [X,Y,Z,WPOIntensity]}
  n146 = Subtract(
   A=MaterialFunctionCall(X=@n145.X, Y=@n145.X, Z=@n145.X) {materialFunction=&MakeFloat3}.Result
   B=ParticlePositionWS()
  )
  n147 = ScalarParameter() {defaultValue=-100, group=Interpolate WPO, parameterName=Clamp_min, sortPriority=1}
  n148 = ScalarParameter() {defaultValue=100, group=Interpolate WPO, parameterName=Clamp_max, sortPriority=2}
  n149 = TextureCoordinate()
  n150 = Power(
   Base=ComponentMask(None=@n149) {r=true}
   Exp=ScalarParameter() {defaultValue=3, group=Interpolate WPO, parameterName=WPOBendPower, sortPriority=3}
  )
  n151 = LinearInterpolate(
   A=Constant()
   B=LinearInterpolate(B=Clamp(None=@n146, Min=@n147, Max=@n148) {maxDefault=1000, minDefault=-1000}, Alpha=Multiply(A=@n145.WPOIntensity, B=@n150))
   Alpha=ScalarParameter() {defaultValue=1, group=Interpolate WPO, parameterName=Use interpolate WPO, sortPriority=0}
  )
  NamedRerouteDeclaration(None=LinearInterpolate(B=@n151, Alpha=ScalarParameter() {parameterName=UseWPO})) {name=Vel Movement to WPO, nodeColor=<rgba 0.4531 1 0 1>}
  NamedRerouteDeclaration(None=@n151) {name=Interpolate WPO, nodeColor=<rgba 0 0.4609 1 1>}
  NamedRerouteUsage()
  OneMinus(None=@n149)
  ParticleRadius()
  Reroute(None=@n127.Opacity)
  ScalarParameter() {parameterName=MakeParticlesFaceZaxis}
  ShadingPathSwitch(Default=@n106, Mobile=Constant()) {desc=Migrated from FeatureLevelSwitch}
  ShadingPathSwitch(Default=@n36, Mobile=Constant3Vector() {constant=<rgba 0 0 1 0>}) {desc=Migrated from FeatureLevelSwitch}
  StaticSwitchParameter(True=@n48, False=@n45) {parameterName=MakeParticleFaceZaxis}
  TextureCoordinate()
  Transform(None=Constant3Vector() {constant=<rgba 0 0 1 0>}) {transformSourceType=TRANSFORMSOURCE_Camera}
 ]
}
